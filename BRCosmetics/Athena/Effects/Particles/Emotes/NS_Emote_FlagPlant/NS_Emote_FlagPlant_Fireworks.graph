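# Niagara fireworks effect for the FlagPlant emote, with looping rocket launches, launch audio and colored bursts.
# FireworkParent shoots flickering sprites upward, colored from a user color curve by spawn group, with front-end-specific velocity, and emits a death event carrying the color ID.
# FireworkSmoke (animated SubUV puffs) and Fireworks_Mesh (expanding randomly oriented mesh with light) use FireworkColor and have no active spawners, suggesting event-driven spawning; handlers are not shown.
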NiagaraSystem /BRCosmetics/Athena/Effects/Particles/Emotes/NS_Emote_FlagPlant/NS_Emote_FlagPlant_Fireworks
dependencies {
 usedDataInterfaces [&NiagaraDataInterfaceCurve, &NiagaraDataInterfaceAudioPlayer]
 usedDynamicInputs []
 usedModules [&SystemState]
}
emitters {
 FireworkParent {
  emitterUpdateScript [
   &EmitterScalability(
    Module.bApplySpawnCountScale=false
    Module.bApplySpawnCountScaleByDistanceFraction=false
    Module.bUseMaxDistance=false
    Module.bUseMinDistance=false
   ) {enabled=false}
   &EmitterLifeCycle(NextLoopDuration=1) {enabled=false}
   &EmitterState(Life Cycle Mode=::NewEnumerator1, Loop Behavior=::NewEnumerator1, Loop Delay=0.1, Loop Duration=2.2, UseLoopDelay=true)
   &SpawnBurst_Instantaneous(Spawn Count=15, Spawn Time=0.05, Use Spawn Probability=false) {enabled=false}
   &SpawnBurst_Instantaneous(Spawn Count=1, Use Spawn Probability=false) {module=SpawnBurst_Instantaneous001}
   &SpawnBurst_Instantaneous(Spawn Count=1, Spawn Group=1, Spawn Time=0.6, Use Spawn Probability=false) {enabled=false, module=SpawnBurst_Instantaneous002}
   &SpawnBurst_Instantaneous(Spawn Count=1, Spawn Group=2, Spawn Time=1, Use Spawn Probability=false) {enabled=false, module=SpawnBurst_Instantaneous003}
   &SpawnRate(SpawnRate=1.7, Use Spawn Probability=false)
  ]
  particleSpawnScript [
   /Game/Effects/Niagara/Modules/Groups/FN_SetSpawnGroup_v1/FN_SetSpawnGroup(Emitter Bone Count=3)
   /SetVariables(Particles.ColorID:NiagaraFloat=&Divide_Float(A=&MakeFloatFromInt(Int=$Particles.SpawnGroup), B=2)) {isSetParameters=true}
   /Niagara/Modules/Spawn/Initialization/InitializeParticle(
    Color=&ColorFromCurve(ColorCurve=$User.FireworkColors, CurveIndex=$Particles.ColorID)
    Lifetime=&UniformRangedFloat(Minimum=0.6, Maximum=0.6)
    Module.Write Color=true
    Module.Write Lifetime=true
    Module.Write Mass=true
    Module.Write Material Random=true
    Module.Write Position=false
    Module.Write Scale=false
    Module.Write SpriteRotation=false
    Module.Write SpriteSize=true
    Position=&SimulationPosition()
    Sprite Size=&Vector2DFromFloat(Value=&UniformRangedFloat(Minimum=24, Maximum=30))
   )
   &AddVelocity(Velocity=&Lerp_Vector(
     A=&UniformRangedVector(Minimum=<-250 -250 700>, Maximum=<250 250 850>)
     B=&UniformRangedVector(Minimum=<-250 -250 600>, Maximum=<250 250 750>)
     Alpha=&FN_FrontEndFloatModifier(Gameplay=0, Front End=1)
    ))
   &ApplyInitialForces()
   &SystemLocation(MaskSpawnGroup=false, Offset=$User.SystemOffset)
   /Niagara/Modules/Audio/PlayAudio(
    Audio Player=DataInterface() {propertyValues {
      Attenuation=/Game/Sounds/Attenuation/Fort_Emotes_Athena_LessOcclusion_Att
      Concurrency=null
      MaxPlaysPerTick=10
      SoundToPlay=/Game/Athena/Sounds/Emotes/CandyApple_Fireworks/CA_Emote_Fireworks_Launch_Cue
      bAllowLoopingOneShotSounds=false
      bLimitPlaysPerTick=true
      bOnlyActiveDuringGameplay=false
      bStopWhenComponentIsDestroyed=true
      ConfigurationUserParameter {parameter {name=None, typeDefHandle {registeredTypeIndex=1}, varData []}}
      ParameterNames []
     }}
    PlayAudio=true
   )
  ]
  particleUpdateScript [
   &ParticleState()
   &Drag(Drag=3, Use Linear Drag=true, Use Rotational Drag=false)
   &SolveForcesAndVelocity(Clamp Velocity=false, Limit Acceleration=false)
   &ScaleColor(
    Scale RGB=&VectorFromFloat(Value=&FloatFromCurve(FloatCurve=DataInterface() {propertyValues {Curve {keys [<at 0 1 RCIM_Cubic>, {arriveTangent=5.0701, interpMode=RCIM_Cubic, leaveTangent=5.0701, tangentMode=RCTM_User, time=1, value=5}]}}}))
    ScaleA=false
    ScaleRGB=true
   )
   &ScaleSpriteSize(Scale Factor=&Vector2DFromFloat(Value=&FloatFromCurve(FloatCurve=DataInterface() {propertyValues {Curve {keys [
          {arriveTangent=5.0716, interpMode=RCIM_Cubic, leaveTangent=5.0716, tangentMode=RCTM_User, time=0, value=0}
          {arriveTangent=0.7499, interpMode=RCIM_Cubic, leaveTangent=0.7499, time=0.1078, value=0.2796}
          <at 0.6168 0.4626>
          <at 0.6485 0.6638>
          <at 0.7101 0.2796>
          <at 0.7567 0.5723>
          <at 0.8108 0.2887>
          <at 0.8574 0.8011>
          <at 0.8947 0.2887>
          <at 0.932 1>
          <at 0.95 0.002>
         ]}}})))
   &GenerateDeathEvent(Float to Send as Random Float:NiagaraFloat=$Particles.Initial.ColorID, Gap Correction Amount=1.5)
  ]
  rendererClasses [&NiagaraSpriteRendererProperties]
  renderers [{rendererClass=&NiagaraSpriteRendererProperties}]
 }
 FireworkSmoke {
  emitterUpdateScript [
   &EmitterScalability(
    Module.bApplySpawnCountScale=false
    Module.bApplySpawnCountScaleByDistanceFraction=false
    Module.bUseMaxDistance=false
    Module.bUseMinDistance=false
   ) {enabled=false}
   &EmitterLifeCycle(NextLoopDuration=1) {enabled=false}
   &EmitterState()
   &SpawnBurst_Instantaneous(Spawn Count=15, Spawn Time=0.05, Use Spawn Probability=false) {enabled=false}
  ]
  particleSpawnScript [
   &InitializeParticle(
    Color=$Particles.FireworkColor
    Lifetime=&UniformRangedFloat(Minimum=0.7, Maximum=1.7)
    Mass=0.1
    Module.Write Color=false
    Module.Write Lifetime=true
    Module.Write Mass=true
    Module.Write Material Random=true
    Module.Write Position=false
    Module.Write Scale=false
    Module.Write SpriteRotation=true
    Module.Write SpriteSize=true
    Position=&SimulationPosition()
    Sprite Rotation=&UniformRangedFloat(Minimum=0, Maximum=360)
    Sprite Size=&Vector2DFromFloat(Value=&UniformRangedFloat(Minimum=80, Maximum=120))
   )
   /SetVariables(Particles.InitSpriteRotRate=&UniformRangedFloat(Minimum=-90, Maximum=90)) {isSetParameters=true}
  ]
  particleUpdateScript [
   &ParticleState()
   &GravityForce(Gravity=<0 0 -35>)
   &SolveForcesAndVelocity(Clamp Velocity=false, Limit Acceleration=false)
   &ScaleColor(
    Scale Alpha=&FloatFromCurve(FloatCurve=DataInterface() {propertyValues {Curve {keys [
         {arriveTangent=11.9154, interpMode=RCIM_Cubic, leaveTangent=11.9154, tangentMode=RCTM_User, time=0, value=0}
         {arriveTangent=0.2004, interpMode=RCIM_Cubic, leaveTangent=0.2004, tangentMode=RCTM_User, time=0.1389, value=0.5438}
         {interpMode=RCIM_Cubic, leaveTangent=-3.3562, time=1.0011, value=0.002}
        ]}}})
    Scale RGB=&Lerp_Vector(
     A=<0.1 0.1 0.1>
     B=&Add_Vector(A=<0.7 0.7 0.7>, B=&MakeVectorFromLinearColorRGB(Color=$Particles.FireworkColor))
     Alpha=&FloatFromCurve(
      FloatCurve=DataInterface() {propertyValues {Curve {keys [{arriveTangent=-2.5161, interpMode=RCIM_Cubic, leaveTangent=-2.5161, tangentMode=RCTM_User, time=0, value=1}, <at 0.7 0>]}}}
      Scale Curve=2
     )
    )
    ScaleA=true
    ScaleRGB=true
   )
   &SpriteRotationRate(Rotation Rate=&Multiply_Float(
     A=$Particles.InitSpriteRotRate
     B=&FloatFromCurve(FloatCurve=DataInterface() {propertyValues {Curve {keys [{arriveTangent=-1.3839, interpMode=RCIM_Cubic, leaveTangent=-1.3839, tangentMode=RCTM_User, time=0, value=1}, <at 1 0.1 RCIM_Cubic>]}}})
    ))
   /Niagara/Modules/Update/SubUV/SubUVAnimation(Number Of Frames=64, SubUV Lookup Index=&Multiply_Float(A=$Particles.NormalizedAge, B=0.63))
   &ScaleSpriteSize(Scale Factor=&Vector2DFromFloat(Value=&FloatFromCurve(FloatCurve=DataInterface() {propertyValues {Curve {keys [
          {arriveTangent=4.9, interpMode=RCIM_Cubic, leaveTangent=4.9, tangentMode=RCTM_User, time=0, value=0}
          {arriveTangent=1.1982, interpMode=RCIM_Cubic, leaveTangent=1.1982, tangentMode=RCTM_User, time=0.1791, value=0.4838}
          {arriveTangent=0.2347, interpMode=RCIM_Cubic, leaveTangent=-1.877, tangentMode=RCTM_User, time=1, value=1}
         ]}}})))
   &DynamicMaterialParameters(
    Index 0 Param 1=&FloatFromCurve(FloatCurve=DataInterface() {propertyValues {Curve {keys [<at 0 1>,<at 1 0>]}}})
    Index 0 Param 2=1
    Index 0 Param 3=1
    Index 0 Param 4=1
    Module.Param0WriteEnabled=true
    Module.Param1WriteEnabled=true
    Module.Param2WriteEnabled=true
    Module.Param3WriteEnabled=true
   )
  ]
  rendererClasses [&NiagaraSpriteRendererProperties]
  renderers [{rendererClass=&NiagaraSpriteRendererProperties}]
 }
 Fireworks_Mesh {
  emitterUpdateScript [
   &EmitterScalability(
    Module.bApplySpawnCountScale=false
    Module.bApplySpawnCountScaleByDistanceFraction=false
    Module.bUseMaxDistance=false
    Module.bUseMinDistance=false
   ) {enabled=false}
   &EmitterLifeCycle(NextLoopDuration=1) {enabled=false}
   &SpawnRate(SpawnRate=20, Use Spawn Probability=false) {enabled=false}
   &EmitterState()
   &SpawnBurst_Instantaneous(Spawn Count=1, Spawn Time=0.05, Use Spawn Probability=false) {enabled=false}
  ]
  particleSpawnScript [
   &InitializeParticle(
    Color=$Particles.FireworkColor
    Lifetime=&UniformRangedFloat(Minimum=1, Maximum=1.3)
    Mesh Scale=&VectorFromFloat(Value=&RandomRangeFloat(Minimum=0.02, Maximum=0.03))
    Module.Write Color=false
    Module.Write Lifetime=true
    Module.Write Mass=true
    Module.Write Material Random=true
    Module.Write Position=true
    Module.Write Scale=true
    Module.Write SpriteRotation=false
    Module.Write SpriteSize=false
    Position=&SimulationPosition()
    Sprite Size=<25 25>
   )
   /Niagara/Modules/Spawn/Orientation/V2/InitialMeshOrientation(
    Orientation Vector=/Niagara/DynamicInputs/Random/RandomVector(
     Vector Scale:NiagaraFloat=1
     Constraint Vector:Vector3f=<0 0 1>
     Randomness Mode:ENiagaraRandomnessMode=::NewEnumerator0
     Random Seed:NiagaraInt32=0
     Fixed Random Seed:NiagaraBool=false
    )
    Use Orientation Vector=true
    Use Rotation Vector=false
   )
  ]
  particleUpdateScript [
   &ParticleState()
   &ScaleColor(
    Color Value To Scale=$Particles.FireworkColor
    Scale Alpha=&FloatFromCurve(FloatCurve=DataInterface() {propertyValues {Curve {keys [
         {arriveTangent=-4.9376, interpMode=RCIM_Cubic, leaveTangent=7.1838, tangentMode=RCTM_User, time=0, value=1}
         {arriveTangent=-2.4774, interpMode=RCIM_Cubic, leaveTangent=-2.4774, tangentMode=RCTM_User, time=0.1389, value=1.3591}
         {arriveTangent=-0.8467, interpMode=RCIM_Cubic, leaveTangent=-0.8467, tangentMode=RCTM_User, time=0.4181, value=0.6281}
         {arriveTangent=-2.194, interpMode=RCIM_Cubic, leaveTangent=-2.1946, tangentMode=RCTM_User, time=1, value=0}
        ]}}})
    Scale RGB=&VectorFromFloat(Value=3)
    ScaleA=true
    ScaleRGB=true
   )
   &ScaleMeshSize(Scale Factor=&VectorFromFloat(Value=&FloatFromCurve(FloatCurve=DataInterface() {propertyValues {Curve {keys [
          {arriveTangent=4.6803, interpMode=RCIM_Cubic, leaveTangent=4.6803, tangentMode=RCTM_User, time=0, value=0}
          {arriveTangent=1.6276, interpMode=RCIM_Cubic, leaveTangent=1.6276, tangentMode=RCTM_User, time=0.1923, value=0.5575}
          {arriveTangent=0.2011, interpMode=RCIM_Cubic, leaveTangent=0.5033, tangentMode=RCTM_User, time=1, value=1}
         ]}}})))
   &GravityForce(Gravity=<0 0 -15>)
   &SolveForcesAndVelocity(Clamp Velocity=false, Limit Acceleration=false)
   &DynamicMaterialParameters(
    Index 0 Param 1=1
    Index 0 Param 2=1
    Index 0 Param 3=&FloatFromCurve(
     FloatCurve=DataInterface() {propertyValues {Curve {keys [
         {arriveTangent=3.0114, interpMode=RCIM_Cubic, leaveTangent=3.0114, tangentMode=RCTM_User, time=0, value=0}
         {arriveTangent=0.9456, interpMode=RCIM_Cubic, leaveTangent=0.9456, tangentMode=RCTM_User, time=0.4343, value=0.7445}
         {arriveTangent=0.3236, interpMode=RCIM_Cubic, leaveTangent=0.7406, tangentMode=RCTM_User, time=1, value=1}
        ]}}}
     Scale Curve=0.2
    )
    Index 0 Param 4=&FloatFromCurve(
     FloatCurve=DataInterface() {propertyValues {Curve {keys [
         {arriveTangent=3.4321, interpMode=RCIM_Cubic, leaveTangent=3.4321, tangentMode=RCTM_User, time=0, value=1}
         {arriveTangent=-1.7957, interpMode=RCIM_Cubic, leaveTangent=-1.7957, tangentMode=RCTM_User, time=0.1672, value=1.5}
         {arriveTangent=-0.286, interpMode=RCIM_Cubic, leaveTangent=-0.286, tangentMode=RCTM_User, time=1, value=0.5}
        ]}}}
     Scale Curve=0.5
    )
    Module.Param0WriteEnabled=true
    Module.Param1WriteEnabled=true
    Module.Param2WriteEnabled=true
    Module.Param3WriteEnabled=true
   )
   &FN_Light_v3(Color=<rgba 1 0 0.7854 1>, Create Color Attribute=true, Create Position Attribute=true, Create Radius Attribute=true, Radius=25)
  ]
  rendererClasses [&NiagaraMeshRendererProperties]
  renderers [{rendererClass=&NiagaraMeshRendererProperties, rendererIndex=0}]
 }
}
summary {
 systemName=NS_Emote_FlagPlant_Fireworks
 emitters [
  {emitterName=Fireworks_Mesh, rendererClasses [&NiagaraMeshRendererProperties]}
  {emitterName=FireworkParent, rendererClasses [&NiagaraSpriteRendererProperties]}
  {emitterName=FireworkSmoke, rendererClasses [&NiagaraSpriteRendererProperties]}
 ]
 userVariables [
  User.bIsFrontEnd {type=NiagaraBool}
  User.bIsFrontEndPreview {type=NiagaraBool}
  User.FireworkColors {
   type=NiagaraDataInterfaceColorCurve
   description=""
   defaultValue {
    struct=NiagaraExt_VariableValue_DataInterface
    value {
     dataInterface=&self:NiagaraDataInterfaceColorCurve_0
     dataInterfaceClass=&NiagaraDataInterfaceColorCurve
    }
   }
  }
  User.SystemOffset {type=Vector3f, description="", defaultValue {struct=Vector3f, value=<0 0 0>}}
 ]
}
